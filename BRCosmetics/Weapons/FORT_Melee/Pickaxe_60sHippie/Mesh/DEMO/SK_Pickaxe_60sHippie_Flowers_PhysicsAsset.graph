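# Physics asset for the flowers skeletal mesh of the 60sHippie two-handed pickaxe cosmetic, in a DEMO mesh folder.
# It holds a single capsule body bound to the Melee_TwoHanded_Handle bone, offset and rotated from that bone.
# There are no constraints, so it provides one simple collision volume with no jointed simulation between parts.
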PhysicsAsset /BRCosmetics/Weapons/FORT_Melee/Pickaxe_60sHippie/Mesh/DEMO/SK_Pickaxe_60sHippie_Flowers_PhysicsAsset
bodies {Melee_TwoHanded_Handle [{shapeType=Capsule, length=63.7461, radius=42.1186, center=<-0.0177 -14.15 25.3938>, rotation=<rot 0 0.2458 -172.8769>}]}
constraints []
